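# Physics asset for the F_MED_BlackMonday female character body, covering dynamic neck ring and cat string accessory bones.
# A box body on the neck ring joins its master bone with twist locked, and that master joins neck_01 with free rotation.
# A separate chain links dyn_cat to two capsule string segments with swing and tight twist limits.
# No constraint connects the cat chain to the neck bodies.
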PhysicsAsset /BRCosmetics/Characters/Player/Female/Medium/Bodies/F_MED_BlackMonday/Meshes/F_MED_BlackMonday_Dyn
bodies {
 dyn_cat [{shapeType=Capsule, length=0.2, radius=0.2, rotation=<rot 90 0 0>}]
 dyn_cat_string_01 [{shapeType=Capsule, length=2, radius=1, center=<1 0 0>, rotation=<rot 90 0 0>}]
 dyn_cat_string_02 [{shapeType=Capsule, length=2, radius=1, center=<1 0 0>, rotation=<rot 90 0 0>}]
 dyn_neck_ring [{shapeType=Box, extentX=3.6141, extentY=3.1158, extentZ=0.5, center=<1.4131 0 0>}]
 dyn_neck_ring_MSTR [{shapeType=Capsule, length=0.505, radius=0.505, rotation=<rot 0 -90 0>}]
 neck_01 [{shapeType=Capsule, length=0.2, radius=0.2, rotation=<rot 90 0 0>}]
}
constraints [
 {
  bone1Name=dyn_neck_ring
  bone2Name=dyn_neck_ring_MSTR
  swing1LimitDegrees=10
  swing2LimitDegrees=30
  twistLimitDegrees=45
  twistMotion=Locked
 }
 {bone1Name=dyn_cat_string_01, bone2Name=dyn_cat, swing1LimitDegrees=40, swing2LimitDegrees=40, twistLimitDegrees=5}
 {bone1Name=dyn_cat_string_02, bone2Name=dyn_cat_string_01, swing1LimitDegrees=45, swing2LimitDegrees=45, twistLimitDegrees=5}
 {
  bone1Name=dyn_neck_ring_MSTR
  bone2Name=neck_01
  swing1LimitDegrees=45
  swing1Motion=Free
  swing2LimitDegrees=45
  swing2Motion=Free
  twistLimitDegrees=45
  twistMotion=Free
 }
]
