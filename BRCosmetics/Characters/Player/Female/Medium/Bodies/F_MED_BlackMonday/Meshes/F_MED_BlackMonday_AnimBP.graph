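# Post-copy Animation Blueprint for the F_MED_BlackMonday female player body, adding corrective and secondary motion to the pose copied from its attached parent mesh.
# Crouch events toggle an alpha driving additive thigh-volume corrections and a dyn_cat offset, the latter interpolated.
# A weighted four-target constraint positions the neck ring from neck, head and upper spine.
# A final RigidBody node simulates the Dyn physics asset relative to spine_03, with scaled component motion.
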AnimBlueprint /BRCosmetics/Characters/Player/Female/Medium/Bodies/F_MED_BlackMonday/Meshes/F_MED_BlackMonday_AnimBP
target_skeleton=&/F_MED_BlackMonday_Skeleton
graphs [
 AnimGraph {nodes [
   n1 = Get crouch alpha() -> (crouch alpha:double)
   n2 = "Transform (Modify) Bone - Bone: thigh_volume_l"(
    ComponentPose=Local To Component(LocalPose=Copy Pose From Mesh() {bUseAttachedParent=true})
    Alpha=@n1
    Translation="-2.362,0.113,0.429"
    Rotation="0,0,0"
    Scale="1.1,1.2,1"
   ) {
    scaleMode=BMM_Additive
    translationMode=BMM_Additive
    translationSpace=BCS_BoneSpace
    alphaBoolBlend {blendOption=Linear}
    boneToModify {boneName=thigh_volume_l}
    scale=<1.1 1.2>
    translation=<-2.362 0.113 0.429>
   }
   n3 = "Transform (Modify) Bone - Bone: thigh_volume_r"(ComponentPose=@n2, Alpha=@n1, Translation="2.362,-0.113,-0.429", Rotation="0,0,0", Scale="1.1,1.2,1") {
    scaleMode=BMM_Additive
    translationMode=BMM_Additive
    translationSpace=BCS_BoneSpace
    alphaBoolBlend {blendOption=Linear}
    boneToModify {boneName=thigh_volume_r}
    scale=<1.1 1.2>
    translation=<2.362 -0.113 -0.429>
   }
   n4 = "Constraint - dyn_neck_ring_MSTR (PPPP)"!(
    ComponentPose=@n3
    Alpha=1
    ConstraintWeights_0=0.2336
    ConstraintWeights_1:float=0.1729
    ConstraintWeights_2:float=0.0814
    ConstraintWeights_3:float=0.5121
   ) -> (Pose) {
    alphaBoolBlend {blendOption=Linear}
    boneToModify {boneName=dyn_neck_ring_MSTR}
    constraintSetup [{targetBone {boneName=neck_02}}, {targetBone {boneName=neck_01}}, {targetBone {boneName=head}}, {targetBone {boneName=spine_05}}]
    constraintWeights [1,1,1,1]
   }
   n5 = "Transform (Modify) Bone - Bone: dyn_cat"(ComponentPose=@n4, Alpha=Get crouch alpha(), Translation="4.509,-5.186,-0.049", Rotation="0.000,-48,0.000", Scale="1,1,1") {
    rotationMode=BMM_Additive
    rotationSpace=BCS_BoneSpace
    translationMode=BMM_Additive
    translationSpace=BCS_BoneSpace
    alphaBoolBlend {blendOption=Linear}
    alphaScaleBiasClamp {bInterpResult=true, interpSpeedDecreasing=12, interpSpeedIncreasing=12}
    boneToModify {boneName=dyn_cat}
    rotation {yaw=-48}
    translation=<4.5088 -5.1862 -0.049>
   }
   n6 = RigidBody(ComponentPose=@n5) {
    bClampLinearTranslationLimitToRefPose=true
    bOverrideWorldGravity=true
    overridePhysicsAsset=&/F_MED_BlackMonday_Dyn
    simulationSpace=BaseBoneSpace
    alphaBoolBlend {blendOption=Linear}
    baseBoneRef {boneName=spine_03}
    componentAppliedLinearAccClamp=<3000 3000 3000>
    componentLinearAccScale=<0.45 0.4 0.4>
    componentLinearVelScale=<0.45 0.4 0.4>
    overrideWorldGravity {z=-980}
   }
   Output Pose(Result=Component To Local(ComponentPose=@n6)) {layerGroup=None}
  ]}
 EventGraph {nodes [
   Event OnBeginCrouching()
   > Set crouch alpha(crouch alpha:double=1) -> (Output_Get:double)
   Event OnEndCrouching()
   > Set crouch alpha(crouch alpha=0)
   Event BlueprintUpdateAnimation()
  ]}
]
variables [crouch alpha {type=number}]
