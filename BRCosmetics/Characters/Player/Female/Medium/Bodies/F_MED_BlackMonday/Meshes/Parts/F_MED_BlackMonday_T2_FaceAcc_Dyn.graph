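# Physics asset for the tier 2 face accessory part of the Black Monday female medium player cosmetic, giving its front hair bones collision and joint limits.
# Capsule bodies cover the head and two front hair bones, dyn_hair_fr_01 and dyn_hair_fr_02.
# Constraints chain the second hair bone to the first and the first to the head, with limited swing and a tight twist range.
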PhysicsAsset /BRCosmetics/Characters/Player/Female/Medium/Bodies/F_MED_BlackMonday/Meshes/Parts/F_MED_BlackMonday_T2_FaceAcc_Dyn
bodies {
 dyn_hair_fr_01 [{shapeType=Capsule, length=1.5, radius=1, center=<1 0 0>, rotation=<rot 90 0 0>}]
 dyn_hair_fr_02 [{shapeType=Capsule, length=1.5, radius=1, center=<1 0 0>, rotation=<rot 90 0 0>}]
 head [{shapeType=Capsule, length=2, radius=2}]
}
constraints [
 {bone1Name=dyn_hair_fr_01, bone2Name=head, swing1LimitDegrees=45, swing2LimitDegrees=20, twistLimitDegrees=5}
 {bone1Name=dyn_hair_fr_02, bone2Name=dyn_hair_fr_01, swing1LimitDegrees=45, swing2LimitDegrees=20, twistLimitDegrees=5}
]
